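# Translucent unlit Niagara material for a glowing energy aura spline, filed under triple tentacle floor attack FX references.
# Squared particle color gets a noise-driven hue shift, while panning noise under sphere masks plus soft glow forms opacity, depth-faded outside low quality and mobile.
# World position offset combines SplineThicken width, sine wobble and a clamped, UV-weighted pull toward a Niagara-set point built from its X value alone.
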Material /PistolCanary/Assets/FX/TripleTentacleFloorAttack/References/M_Aura_Energy_Glow_Spline
graph {
 properties {
  bUsedWithNiagaraMeshParticles=true
  bUsedWithNiagaraSprites=true
  blendMode=BLEND_Translucent
  pixelDepthOffsetMode=PDOM_Legacy
  shadingModel=MSM_Unlit
  usedShadingModels=Unlit
 }
 nodes [
  n1 = Add(A=TextureCoordinate() {uTiling=0.35, vTiling=0.35}, B=ParticleRandom())
  n2 = AppendVector(
   A=ScalarParameter() {defaultValue=1, group=HueShift, parameterName=Hue Shift Texture X Scale, sortPriority=3}
   B=ScalarParameter() {defaultValue=1, group=HueShift, parameterName=Hue Shift Texture Y Scale, sortPriority=4}
  )
  n3 = ScalarParameter() {group=HueShift, parameterName=Hue Shift Texture X Speed, sortPriority=5}
  n4 = ScalarParameter() {defaultValue=0.5, group=HueShift, parameterName=Hue Shift Texture Y Speed, sortPriority=6}
  n5 = TextureSampleParameter2D(UVs=Panner(Coordinate=Multiply(A=@n1, B=@n2), Speed=AppendVector(A=@n3, B=@n4)) {speedY=0.5}) {
   group=HueShift
   parameterName=TextureHueShift
   samplerType=SAMPLERTYPE_Grayscale
   sortPriority=1
   texture=&LowResBlurredNoise
  }
  n6 = Multiply(A=@n5.R, B=ScalarParameter() {defaultValue=-0.5, group=HueShift, parameterName=TextureHueShiftAmount, sortPriority=2})
  n7 = ParticleColor()
  n8 = Multiply(A=@n7.RGB, B=@n7.RGB)
  MaterialFunctionCall(Emissive=StaticSwitchParameter(
    True=MaterialFunctionCall(Hue Shift Percentage=@n6, Texture=@n8) {materialFunction=&HueShift}.Result
    False=@n8
   ) {defaultValue=true, group=HueShift, parameterName=UseTextureHueShift, sortPriority=0}) {materialFunction=&MF_FX_EmissiveAdjustments}
  n9 = DynamicParameter() {desc=Parameters set by Niagara, parameterIndex=3, defaultValue=<rgba 0 1 1 1>, paramNames [X,Y,Z,Intensity]}
  n10 = MaterialFunctionCall(Float3=Subtract(
    A=MaterialFunctionCall(X=@n9.X, Y=@n9.X, Z=@n9.X) {materialFunction=&MakeFloat3}.Result
    B=ParticlePositionWS()
   )) {materialFunction=&BreakOutFloat3Components}
  n11 = ScalarParameter() {defaultValue=100, desc=Interpolate Clamp XY, group=WPO, parameterName=Interpolate XY Clamp, sortPriority=3}
  n12 = ScalarParameter() {defaultValue=2, desc=Interpolate Clamp Z, group=WPO, parameterName=Interpolate Clamp Z, sortPriority=4}
  n13 = AppendVector(
   A=Clamp(None=AppendVector(A=@n10.R, B=@n10.R), Min=Multiply(A=@n11) {constB=-1}, Max=@n11) {maxDefault=1000, minDefault=-1000}
   B=Clamp(None=@n10.B, Min=Multiply(A=@n12) {constB=-1}, Max=@n12) {minDefault=-1}
  )
  n14 = OneMinus(None=TextureCoordinate())
  n15 = Multiply(
   A=@n9.Intensity
   B=Power(
    Base=ComponentMask(None=@n14) {g=true}
    Exp=ScalarParameter() {defaultValue=5, desc=Interpolate Power, group=WPO, parameterName=Interpolate Power, sortPriority=2}
   )
  )
  n16 = DynamicParameter() {defaultValue=<rgba -60 1 1 1>, paramNames [Width,Param2,Param3,Param4]}
  n17 = Add(
   A=StaticSwitchParameter(True=LinearInterpolate(B=@n13, Alpha=@n15), False=Constant()) {defaultValue=true, group=WPO, parameterName=Use Interpolate WPO, sortPriority=1}
   B=MaterialFunctionCall(WidthBase=@n16.Width, WidthTip=@n16.Width) {materialFunction=&SplineThicken}.WorldPositionOffset
  )
  n18 = ComponentMask(None=TextureCoordinate()) {g=true}
  n19 = Multiply(A=Add(A=Time(), B=ParticleRandom()), B=LinearInterpolate(Alpha=ParticleRandom()) {constA=0.7, constB=1.3})
  n20 = Sine(None=Multiply(A=Add(A=@n18, B=@n19), B=ScalarParameter() {defaultValue=0.9, group=WPO, parameterName=WPOSineAmount, sortPriority=10}))
  n21 = Constant3Vector() {constant=<rgba 0.75 0.75 0 1>}
  n22 = TextureCoordinate()
  n23 = Multiply(
   A=Multiply(A=Multiply(A=@n20, B=@n21), B=ScalarParameter() {defaultValue=15, group=WPO, parameterName=WPOSineIntensity, sortPriority=11})
   B=OneMinus(None=ComponentMask(None=@n22) {g=true})
  )
  NamedRerouteDeclaration(None=StaticSwitchParameter(True=Add(A=@n17, B=@n23), False=@n17) {group=WPO, parameterName=Use Wobble WPO, sortPriority=9}) {name=WPO, nodeColor=<rgba 0 0 0 1>}
  n24 = Add(
   A=TextureCoordinate()
   B=AppendVector(
    A=ScalarParameter() {defaultValue=0.15, group=Texture, parameterName=Texture X Offset, sortPriority=5}
    B=ScalarParameter() {defaultValue=0.15, group=Texture, parameterName=Texture Y Offset, sortPriority=6}
   )
  )
  n25 = AppendVector(
   A=ScalarParameter() {defaultValue=0.6, group=Texture, parameterName=Texture X Scale, sortPriority=3}
   B=ScalarParameter() {defaultValue=0.75, group=Texture, parameterName=Texture Y Scale, sortPriority=4}
  )
  n26 = Time()
  n27 = ParticleRandom()
  n28 = ScalarParameter() {group=Texture, parameterName=Texture X Speed, sortPriority=1}
  n29 = ScalarParameter() {defaultValue=1, group=Texture, parameterName=Texture Y Speed, sortPriority=2}
  n30 = TextureSampleParameter2D(UVs=Panner(Coordinate=Multiply(A=@n24, B=@n25), Time=Add(A=@n26, B=@n27), Speed=AppendVector(A=@n28, B=@n29)) {speedY=1}) {
   group=Texture
   parameterName=Texture
   samplerType=SAMPLERTYPE_LinearGrayscale
   sortPriority=0
   texture=/Game/Effects/Fort_Effects/Textures/T_DarkAndLight_Noise
  }
  n31 = OneMinus(None=@n30.R)
  n32 = Constant() {r=0.1}
  n33 = SphereMask(A=Multiply(A=TextureCoordinate(), B=Constant2Vector() {g=0.5, r=1}), B=Constant() {r=0.5}, Radius=Constant() {r=0.5}, Hardness=Constant()) {hardnessPercent=0}
  n34 = Multiply(
   A=Add(A=StaticSwitchParameter(True=@n31, False=@n30.R) {group=Texture, parameterName=Invert Texture, sortPriority=0}, B=Multiply(A=@n32, B=@n33))
   B=@n33
  )
  n35 = ScalarParameter() {defaultValue=0.2, group=Texture, parameterName=Texture Subtract, sortPriority=7}
  n36 = Saturate(None=Divide(A=Subtract(A=@n34, B=@n35) {constB=0.2}, B=ScalarParameter() {defaultValue=0.5, group=Texture, parameterName=Texture Divide, sortPriority=8}) {constB=1})
  n37 = Power(Base=SphereMask(A=TextureCoordinate(), B=Constant() {r=0.5}, Radius=Constant() {r=0.5}, Hardness=Constant()) {hardnessPercent=0})
  n38 = Multiply(A=@n37, B=ScalarParameter() {defaultValue=0.2, group=Texture, parameterName=Glow Amount}) {constB=0.2}
  n39 = Multiply(A=@n7.A, B=Add(A=Multiply(A=@n36, B=@n37), B=Saturate(None=@n38)) {desc=Add some subtle glow})
  n40 = DepthFade(Opacity=@n39, FadeDistance=ScalarParameter() {defaultValue=25, group=Opacity, parameterName=DepthFade, sortPriority=0})
  ShadingPathSwitch(Default=QualitySwitch(Default=@n40, Low=@n39, High=@n40, Medium=@n40, Epic=@n40), Mobile=@n39)
  StaticSwitchParameter(True=Constant(), False=NamedRerouteUsage()) {group=WPO, parameterName=DisableWPO, sortPriority=0}
 ]
}
